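# Physics asset for the shirt cloth bones on the right side of the SteamPower Ice medium male body cosmetic.
# Five tiny base_shirt anchor capsules attach to the pelvis with free swing and twist, and each carries a larger dyn_shirt capsule body on a per-bone angular limit.
# Limits generally tighten from chain 01 to 05, with zero twist range on the last three.
# Large pelvis and right thigh capsules provide body volume beside the cloth.
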PhysicsAsset /BRCosmetics/Characters/Player/Male/Medium/Bodies/M_MED_SteamPower_Ice/Meshes/M_MED_SteamPower_Ice_Cloth_Physics
bodies {
 base_shirt_01_r [{shapeType=Capsule, length=0.505, radius=0.505, shapeName=base_shirt_01_r_capsule, rotation=<rot 0 -90 0>}]
 base_shirt_02_r [{shapeType=Capsule, length=0.505, radius=0.505, shapeName=base_shirt_02_r_capsule, rotation=<rot 0 -90 0>}]
 base_shirt_03_r [{shapeType=Capsule, length=0.505, radius=0.505, shapeName=base_shirt_03_r_capsule, rotation=<rot 0 -90 0>}]
 base_shirt_04_r [{shapeType=Capsule, length=0.505, radius=0.505, shapeName=base_shirt_04_r_capsule, rotation=<rot 0 -90 0>}]
 base_shirt_05_r [{shapeType=Capsule, length=0.505, radius=0.505, shapeName=base_shirt_05_r_capsule, rotation=<rot 0 -90 0>}]
 dyn_shirt_01_r [
  {
   shapeType=Capsule
   length=5.9836
   radius=0.9
   shapeName=dyn_shirt_01_r_capsule
   center=<3.0227 0.5461 0.2284>
   rotation=<rot -0.0001 90 -89.9998>
  }
  {
   shapeType=Capsule
   length=3.4836
   radius=0.9
   shapeName=dyn_shirt_01_r_capsule
   center=<3.1997 3.6402 -0.3213>
   rotation=<rot -0.0001 90 -89.9998>
  }
 ]
 dyn_shirt_02_r [{
   shapeType=Capsule
   length=6.243
   radius=1.6728
   shapeName=dyn_shirt_02_r_capsule
   center=<4.3745 -2.1293 0.6935>
   rotation=<rot 1.7282 45.1488 99.8513>
  }]
 dyn_shirt_03_r [{shapeType=Capsule, length=5.3454, radius=1.174, shapeName=dyn_shirt_03_r_capsule, center=<2.1249 0 0>, rotation=<rot 0 90 -90>}]
 dyn_shirt_04_r [{shapeType=Capsule, length=2.9104, radius=0.9167, shapeName=dyn_shirt_04_r_capsule, center=<1.9894 0 0>, rotation=<rot 0 90 90>}]
 dyn_shirt_05_r [{shapeType=Capsule, length=2.9061, radius=0.9899, shapeName=dyn_shirt_05_r_capsule, center=<2 0 0>, rotation=<rot 0 90 -90>}]
 pelvis [{shapeType=Capsule, length=8.9475, radius=9.5755, center=<0 -1.4954 0>}]
 thigh_r [{shapeType=Capsule, length=36.6238, radius=9.3643, center=<22.111 1.6086 -1.3893>, rotation=<rot 0.2332 -85.9391 -93.8921>}]
}
constraints [
 {bone1Name=thigh_r, bone2Name=pelvis, swing1LimitDegrees=45, swing2LimitDegrees=45, twistLimitDegrees=45}
 {bone1Name=dyn_shirt_01_r, bone2Name=base_shirt_01_r, swing1LimitDegrees=25, swing2LimitDegrees=65, twistLimitDegrees=20}
 {bone1Name=dyn_shirt_02_r, bone2Name=base_shirt_02_r, swing1LimitDegrees=15, swing2LimitDegrees=65, twistLimitDegrees=10}
 {bone1Name=dyn_shirt_03_r, bone2Name=base_shirt_03_r, swing1LimitDegrees=25, swing2LimitDegrees=30, twistLimitDegrees=0}
 {
  bone1Name=base_shirt_04_r
  bone2Name=pelvis
  swing1LimitDegrees=45
  swing1Motion=Free
  swing2LimitDegrees=45
  swing2Motion=Free
  twistLimitDegrees=45
  twistMotion=Free
 }
 {
  bone1Name=base_shirt_03_r
  bone2Name=pelvis
  swing1LimitDegrees=45
  swing1Motion=Free
  swing2LimitDegrees=45
  swing2Motion=Free
  twistLimitDegrees=45
  twistMotion=Free
 }
 {
  bone1Name=base_shirt_02_r
  bone2Name=pelvis
  swing1LimitDegrees=45
  swing1Motion=Free
  swing2LimitDegrees=45
  swing2Motion=Free
  twistLimitDegrees=45
  twistMotion=Free
 }
 {
  bone1Name=base_shirt_01_r
  bone2Name=pelvis
  swing1LimitDegrees=45
  swing1Motion=Free
  swing2LimitDegrees=45
  swing2Motion=Free
  twistLimitDegrees=45
  twistMotion=Free
 }
 {bone1Name=dyn_shirt_04_r, bone2Name=base_shirt_04_r, swing1LimitDegrees=10, swing2LimitDegrees=30, twistLimitDegrees=0}
 {
  bone1Name=base_shirt_05_r
  bone2Name=pelvis
  swing1LimitDegrees=45
  swing1Motion=Free
  swing2LimitDegrees=45
  swing2Motion=Free
  twistLimitDegrees=45
  twistMotion=Free
 }
 {bone1Name=dyn_shirt_05_r, bone2Name=base_shirt_05_r, swing1LimitDegrees=5, swing2LimitDegrees=20, twistLimitDegrees=0}
]
